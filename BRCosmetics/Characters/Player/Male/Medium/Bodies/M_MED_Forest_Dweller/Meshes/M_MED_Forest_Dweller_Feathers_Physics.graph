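# Physics asset for the feathers on the medium male Forest Dweller player body, covering left, right and back feather bones.
# Each feather is a two-bone pair, with a small box on the first bone and a capsule on the second.
# One constraint joins each pair with tight swing and twist limits; a few left and back feathers get a wider second swing.
# No constraints link separate feathers or other body bones.
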PhysicsAsset /BRCosmetics/Characters/Player/Male/Medium/Bodies/M_MED_Forest_Dweller/Meshes/M_MED_Forest_Dweller_Feathers_Physics
bodies {
 dyn_bck_feathA_01 [{shapeType=Box, extentX=1.01, extentY=1.01, extentZ=1.01, rotation=<rot 0 -90 0>}]
 dyn_bck_feathA_02 [{shapeType=Capsule, length=4, radius=1.65, center=<2 0 0>, rotation=<rot 90 0 0>}]
 dyn_bck_feathB_01 [{shapeType=Box, extentX=1.01, extentY=1.01, extentZ=1.01, rotation=<rot 0 -90 0>}]
 dyn_bck_feathB_02 [{shapeType=Capsule, length=4, radius=1.65, center=<2.25 0 0>, rotation=<rot 90 0 0>}]
 dyn_bck_feathC_01 [{shapeType=Box, extentX=1.01, extentY=1.01, extentZ=1.01, rotation=<rot 0 -90 0>}]
 dyn_bck_feathC_02 [{shapeType=Capsule, length=3, radius=1.15, center=<2 0 0>, rotation=<rot 90 0 0>}]
 dyn_bck_feathD_01 [{shapeType=Box, extentX=1.01, extentY=1.01, extentZ=1.01, rotation=<rot 0 -90 0>}]
 dyn_bck_feathD_02 [{shapeType=Capsule, length=4, radius=1.65, center=<2 0 0>, rotation=<rot 90 0 0>}]
 dyn_bck_feathE_01 [{shapeType=Box, extentX=1.01, extentY=1.01, extentZ=1.01, rotation=<rot 0 -90 0>}]
 dyn_bck_feathE_02 [{shapeType=Capsule, length=4.75, radius=0.9, center=<1.3629 -0.0001 0>, rotation=<rot 90 0 0>}]
 dyn_feathA_01_l [{shapeType=Box, extentX=1.01, extentY=1.01, extentZ=1.01, rotation=<rot 0 -90 0>}]
 dyn_feathA_01_r [{shapeType=Box, extentX=1.01, extentY=1.01, extentZ=1.01, rotation=<rot 0 -90 0>}]
 dyn_feathA_02_l [{shapeType=Capsule, length=3, radius=1.15, center=<2 0 0>, rotation=<rot 90 0 0>}]
 dyn_feathA_02_r [{shapeType=Capsule, length=4, radius=1.65, center=<2 0 0>, rotation=<rot 90 0 0>}]
 dyn_feathB_01_l [{shapeType=Box, extentX=1.01, extentY=1.01, extentZ=1.01, rotation=<rot 0 -90 0>}]
 dyn_feathB_01_r [{shapeType=Box, extentX=1.01, extentY=1.01, extentZ=1.01, rotation=<rot 0 -90 0>}]
 dyn_feathB_02_l [{shapeType=Capsule, length=4, radius=1.65, center=<2 0 0>, rotation=<rot 90 0 0>}]
 dyn_feathB_02_r [{shapeType=Capsule, length=3, radius=1.15, center=<2 0 0>, rotation=<rot 90 0 0>}]
 dyn_feathC_01_l [{shapeType=Box, extentX=1.01, extentY=1.01, extentZ=1.01, rotation=<rot 0 -90 0>}]
 dyn_feathC_01_r [{shapeType=Box, extentX=1.01, extentY=1.01, extentZ=1.01, rotation=<rot 0 -90 0>}]
 dyn_feathC_02_l [{shapeType=Capsule, length=3, radius=1.15, center=<2 0 0>, rotation=<rot 90 0 0>}]
 dyn_feathC_02_r [{shapeType=Capsule, length=3, radius=1.15, center=<2 0 0>, rotation=<rot 90 0 0>}]
 dyn_feathD_01_l [{shapeType=Box, extentX=1.01, extentY=1.01, extentZ=1.01, rotation=<rot 0 -90 0>}]
 dyn_feathD_01_r [{shapeType=Box, extentX=1.01, extentY=1.01, extentZ=1.01, rotation=<rot 0 -90 0>}]
 dyn_feathD_02_l [{shapeType=Capsule, length=4, radius=1.65, center=<2 0 0>, rotation=<rot 90 0 0>}]
 dyn_feathD_02_r [{shapeType=Capsule, length=4, radius=1.65, center=<2 0 0>, rotation=<rot 90 0 0>}]
 dyn_feathE_01_l [{shapeType=Box, extentX=1.01, extentY=1.01, extentZ=1.01, rotation=<rot 0 -90 0>}]
 dyn_feathE_01_r [{shapeType=Box, extentX=1.01, extentY=1.01, extentZ=1.01, rotation=<rot 0 -90 0>}]
 dyn_feathE_02_l [{shapeType=Capsule, length=4, radius=1.65, center=<2 0 0>, rotation=<rot 90 0 0>}]
 dyn_feathE_02_r [{shapeType=Capsule, length=4, radius=1.65, center=<2 0 0>, rotation=<rot 90 0 0>}]
 dyn_feathF_01_l [{shapeType=Box, extentX=1.01, extentY=1.01, extentZ=1.01, rotation=<rot 0 -90 0>}]
 dyn_feathF_01_r [{shapeType=Box, extentX=1.01, extentY=1.01, extentZ=1.01, rotation=<rot 0 -90 0>}]
 dyn_feathF_02_l [{shapeType=Capsule, length=3, radius=1.15, center=<2 0 0>, rotation=<rot 90 0 0>}]
 dyn_feathF_02_r [{shapeType=Capsule, length=3, radius=1.15, center=<2 0 0>, rotation=<rot 90 0 0>}]
 dyn_feathG_01_l [{shapeType=Box, extentX=1.01, extentY=1.01, extentZ=1.01, rotation=<rot 0 -90 0>}]
 dyn_feathG_01_r [{shapeType=Box, extentX=1.01, extentY=1.01, extentZ=1.01, rotation=<rot 0 -90 0>}]
 dyn_feathG_02_l [{shapeType=Capsule, length=3, radius=1.15, center=<2 0 0>, rotation=<rot 90 0 0>}]
 dyn_feathG_02_r [{shapeType=Capsule, length=3, radius=1.15, center=<2 0 0>, rotation=<rot 90 0 0>}]
 dyn_feathH_01_l [{shapeType=Box, extentX=1.01, extentY=1.01, extentZ=1.01, rotation=<rot 0 -90 0>}]
 dyn_feathH_01_r [{shapeType=Box, extentX=1.01, extentY=1.01, extentZ=1.01, rotation=<rot 0 -90 0>}]
 dyn_feathH_02_l [{shapeType=Capsule, length=3.25, radius=1.15, center=<1.3579 0.0001 0>, rotation=<rot 90 0 0>}]
 dyn_feathH_02_r [{shapeType=Capsule, length=3, radius=1.15, center=<2 0 0>, rotation=<rot 90 0 0>}]
 dyn_feathI_01_l [{shapeType=Box, extentX=1.01, extentY=1.01, extentZ=1.01, rotation=<rot 0 -90 0>}]
 dyn_feathI_01_r [{shapeType=Box, extentX=1.01, extentY=1.01, extentZ=1.01, rotation=<rot 0 -90 0>}]
 dyn_feathI_02_l [{shapeType=Capsule, length=3, radius=1.15, center=<2 0 0>, rotation=<rot 90 0 0>}]
 dyn_feathI_02_r [{shapeType=Capsule, length=4, radius=1.65, center=<2 0 0>, rotation=<rot 90 0 0>}]
 dyn_feathJ_01_l [{shapeType=Box, extentX=1.01, extentY=1.01, extentZ=1.01, rotation=<rot 0 -90 0>}]
 dyn_feathJ_01_r [{shapeType=Box, extentX=1.01, extentY=1.01, extentZ=1.01, rotation=<rot 0 -90 0>}]
 dyn_feathJ_02_l [{shapeType=Capsule, length=4, radius=1.65, center=<2 0 0>, rotation=<rot 90 0 0>}]
 dyn_feathJ_02_r [{shapeType=Capsule, length=3, radius=1.15, center=<2 0 0>, rotation=<rot 90 0 0>}]
 dyn_feathK_01_r [{shapeType=Box, extentX=1.01, extentY=1.01, extentZ=1.01, rotation=<rot 0 -90 0>}]
 dyn_feathK_02_r [{shapeType=Capsule, length=6, radius=1.9, center=<2.5807 -0.0001 0>, rotation=<rot 90 0 0>}]
}
constraints [
 {bone1Name=dyn_feathA_02_r, bone2Name=dyn_feathA_01_r, swing1LimitDegrees=15, swing2LimitDegrees=15, twistLimitDegrees=5}
 {bone1Name=dyn_feathB_02_r, bone2Name=dyn_feathB_01_r, swing1LimitDegrees=15, swing2LimitDegrees=15, twistLimitDegrees=5}
 {bone1Name=dyn_feathC_02_r, bone2Name=dyn_feathC_01_r, swing1LimitDegrees=15, swing2LimitDegrees=15, twistLimitDegrees=5}
 {bone1Name=dyn_feathD_02_r, bone2Name=dyn_feathD_01_r, swing1LimitDegrees=15, swing2LimitDegrees=15, twistLimitDegrees=5}
 {bone1Name=dyn_feathE_02_r, bone2Name=dyn_feathE_01_r, swing1LimitDegrees=15, swing2LimitDegrees=15, twistLimitDegrees=5}
 {bone1Name=dyn_feathF_02_r, bone2Name=dyn_feathF_01_r, swing1LimitDegrees=15, swing2LimitDegrees=15, twistLimitDegrees=5}
 {bone1Name=dyn_feathG_02_r, bone2Name=dyn_feathG_01_r, swing1LimitDegrees=15, swing2LimitDegrees=15, twistLimitDegrees=5}
 {bone1Name=dyn_feathH_02_r, bone2Name=dyn_feathH_01_r, swing1LimitDegrees=15, swing2LimitDegrees=15, twistLimitDegrees=5}
 {bone1Name=dyn_feathI_02_r, bone2Name=dyn_feathI_01_r, swing1LimitDegrees=15, swing2LimitDegrees=15, twistLimitDegrees=5}
 {bone1Name=dyn_feathJ_02_r, bone2Name=dyn_feathJ_01_r, swing1LimitDegrees=15, swing2LimitDegrees=15, twistLimitDegrees=5}
 {bone1Name=dyn_feathK_02_r, bone2Name=dyn_feathK_01_r, swing1LimitDegrees=15, swing2LimitDegrees=15, twistLimitDegrees=5}
 {bone1Name=dyn_feathA_02_l, bone2Name=dyn_feathA_01_l, swing1LimitDegrees=15, swing2LimitDegrees=15, twistLimitDegrees=5}
 {bone1Name=dyn_feathB_02_l, bone2Name=dyn_feathB_01_l, swing1LimitDegrees=15, swing2LimitDegrees=15, twistLimitDegrees=5}
 {bone1Name=dyn_feathC_02_l, bone2Name=dyn_feathC_01_l, swing1LimitDegrees=15, swing2LimitDegrees=15, twistLimitDegrees=5}
 {bone1Name=dyn_feathD_02_l, bone2Name=dyn_feathD_01_l, swing1LimitDegrees=15, swing2LimitDegrees=15, twistLimitDegrees=5}
 {bone1Name=dyn_feathE_02_l, bone2Name=dyn_feathE_01_l, swing1LimitDegrees=15, swing2LimitDegrees=15, twistLimitDegrees=5}
 {bone1Name=dyn_feathF_02_l, bone2Name=dyn_feathF_01_l, swing1LimitDegrees=15, swing2LimitDegrees=15, twistLimitDegrees=5}
 {bone1Name=dyn_feathG_02_l, bone2Name=dyn_feathG_01_l, swing1LimitDegrees=15, swing2LimitDegrees=15, twistLimitDegrees=5}
 {bone1Name=dyn_feathH_02_l, bone2Name=dyn_feathH_01_l, swing1LimitDegrees=15, swing2LimitDegrees=21.8571, twistLimitDegrees=5}
 {bone1Name=dyn_feathI_02_l, bone2Name=dyn_feathI_01_l, swing1LimitDegrees=15, swing2LimitDegrees=30, twistLimitDegrees=5}
 {bone1Name=dyn_feathJ_02_l, bone2Name=dyn_feathJ_01_l, swing1LimitDegrees=15, swing2LimitDegrees=15, twistLimitDegrees=5}
 {bone1Name=dyn_bck_feathA_02, bone2Name=dyn_bck_feathA_01, swing1LimitDegrees=15, swing2LimitDegrees=15, twistLimitDegrees=5}
 {bone1Name=dyn_bck_feathB_02, bone2Name=dyn_bck_feathB_01, swing1LimitDegrees=15, swing2LimitDegrees=25, twistLimitDegrees=5}
 {bone1Name=dyn_bck_feathC_02, bone2Name=dyn_bck_feathC_01, swing1LimitDegrees=15, swing2LimitDegrees=15, twistLimitDegrees=5}
 {bone1Name=dyn_bck_feathD_02, bone2Name=dyn_bck_feathD_01, swing1LimitDegrees=15, swing2LimitDegrees=15, twistLimitDegrees=5}
 {bone1Name=dyn_bck_feathE_02, bone2Name=dyn_bck_feathE_01, swing1LimitDegrees=15, swing2LimitDegrees=22, twistLimitDegrees=5}
]
