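# Physics asset for the male medium Forest Dweller cosmetic body, covering its dynamic cloak bones and upper spine.
# Three chains of three box bodies (back, left, right) hang from small parent bodies jointed to spine_03 with free rotation, alongside two spine capsules.
# Cloak links use tight twist limits, except the left chain's first link, which rotates freely, unlike its back and right counterparts.
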PhysicsAsset /BRCosmetics/Characters/Player/Male/Medium/Bodies/M_MED_Forest_Dweller/Meshes/M_MED_Forest_Dweller_Physics
bodies {
 dyn_cloak_bck_01 [{shapeType=Box, extentX=1.2376, extentY=13.9066, extentZ=8.4649, center=<3 0 0>, rotation=<rot 0 90 -90>}]
 dyn_cloak_bck_01_l [{shapeType=Box, extentX=1.2376, extentY=13.9066, extentZ=8.4649, center=<3 0 0>, rotation=<rot 0 90 -90>}]
 dyn_cloak_bck_01_l_r [{shapeType=Box, extentX=1.2376, extentY=13.9066, extentZ=8.4649, center=<3 0 0>, rotation=<rot 0 90 -90>}]
 dyn_cloak_bck_02 [{shapeType=Box, extentX=1.01, extentY=12.01, extentZ=13.51, center=<3 0 0>, rotation=<rot 0 90 -90>}]
 dyn_cloak_bck_02_l [{shapeType=Box, extentX=1.01, extentY=12.01, extentZ=13.51, center=<3 0 0>, rotation=<rot 0 90 -90>}]
 dyn_cloak_bck_02_l_r [{shapeType=Box, extentX=1.01, extentY=12.01, extentZ=13.51, center=<3 0 0>, rotation=<rot 0 90 -90>}]
 dyn_cloak_bck_03 [{shapeType=Box, extentX=2.51, extentY=7.51, extentZ=11.01, center=<7 0 0>, rotation=<rot 0 90 -90>}]
 dyn_cloak_bck_03_l [{shapeType=Box, extentX=2.51, extentY=7.51, extentZ=11.01, center=<7 0 0>, rotation=<rot 0 90 -90>}]
 dyn_cloak_bck_03_l_r [{shapeType=Box, extentX=2.51, extentY=7.51, extentZ=11.01, center=<7 0 0>, rotation=<rot 0 90 -90>}]
 dyn_parent_cloak_bck [{shapeType=Capsule, length=0.505, radius=0.505, rotation=<rot 0 -90 0>}]
 dyn_parent_cloak_l [{shapeType=Box, extentX=1.01, extentY=1.01, extentZ=1.01, rotation=<rot 0 -90 0>}]
 dyn_parent_cloak_r [{shapeType=Box, extentX=1.01, extentY=1.01, extentZ=1.01, rotation=<rot 0 -90 0>}]
 spine_01 [{shapeType=Capsule, length=7.9374, radius=12.1041, center=<5.3781 1.334 0.0301>, rotation=<rot -0.8694 -90 -1.447>}]
 spine_03 [{shapeType=Capsule, length=10.75, radius=13.25, center=<8.0428 6.0422 -0.0001>, rotation=<rot 90 0 -2.3564>}]
}
constraints [
 {bone1Name=spine_03, bone2Name=spine_01, swing1LimitDegrees=45, swing2LimitDegrees=45, twistLimitDegrees=45}
 {
  bone1Name=dyn_parent_cloak_bck
  bone2Name=spine_03
  swing1LimitDegrees=45
  swing1Motion=Free
  swing2LimitDegrees=45
  swing2Motion=Free
  twistLimitDegrees=45
  twistMotion=Free
 }
 {bone1Name=dyn_cloak_bck_01, bone2Name=dyn_parent_cloak_bck, swing1LimitDegrees=45, swing2LimitDegrees=45, twistLimitDegrees=8}
 {bone1Name=dyn_cloak_bck_02, bone2Name=dyn_cloak_bck_01, swing1LimitDegrees=45, swing2LimitDegrees=45, twistLimitDegrees=8}
 {bone1Name=dyn_cloak_bck_03, bone2Name=dyn_cloak_bck_02, swing1LimitDegrees=45, swing2LimitDegrees=45, twistLimitDegrees=8}
 {
  bone1Name=dyn_parent_cloak_l
  bone2Name=spine_03
  swing1LimitDegrees=45
  swing1Motion=Free
  swing2LimitDegrees=45
  swing2Motion=Free
  twistLimitDegrees=45
  twistMotion=Free
 }
 {
  bone1Name=dyn_parent_cloak_r
  bone2Name=spine_03
  swing1LimitDegrees=45
  swing1Motion=Free
  swing2LimitDegrees=45
  swing2Motion=Free
  twistLimitDegrees=45
  twistMotion=Free
 }
 {
  bone1Name=dyn_cloak_bck_01_l
  bone2Name=dyn_parent_cloak_l
  swing1LimitDegrees=45
  swing1Motion=Free
  swing2LimitDegrees=45
  swing2Motion=Free
  twistLimitDegrees=45
  twistMotion=Free
 }
 {bone1Name=dyn_cloak_bck_02_l, bone2Name=dyn_cloak_bck_01_l, swing1LimitDegrees=45, swing2LimitDegrees=45, twistLimitDegrees=8}
 {bone1Name=dyn_cloak_bck_03_l, bone2Name=dyn_cloak_bck_02_l, swing1LimitDegrees=45, swing2LimitDegrees=45, twistLimitDegrees=8}
 {bone1Name=dyn_cloak_bck_01_l_r, bone2Name=dyn_parent_cloak_r, swing1LimitDegrees=45, swing2LimitDegrees=45, twistLimitDegrees=8}
 {bone1Name=dyn_cloak_bck_02_l_r, bone2Name=dyn_cloak_bck_01_l_r, swing1LimitDegrees=45, swing2LimitDegrees=45, twistLimitDegrees=8}
 {bone1Name=dyn_cloak_bck_03_l_r, bone2Name=dyn_cloak_bck_02_l_r, swing1LimitDegrees=45, swing2LimitDegrees=45, twistLimitDegrees=8}
]
