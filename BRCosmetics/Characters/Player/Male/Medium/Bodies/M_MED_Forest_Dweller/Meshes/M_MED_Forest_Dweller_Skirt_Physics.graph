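# Physics asset that simulates the skirt of the male medium Forest Dweller player body.
# Capsules on pelvis, thighs and calves form the body collision shapes.
# Four skirt chains (front/back, left/right) each run from a sphere base, constrained to the thigh with free rotation, through two box segments with limited swing and twist.
# The upper back panels are angularly locked together, and the lower back-right segment adds an extra capsule.
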PhysicsAsset /BRCosmetics/Characters/Player/Male/Medium/Bodies/M_MED_Forest_Dweller/Meshes/M_MED_Forest_Dweller_Skirt_Physics
bodies {
 calf_l [{shapeType=Capsule, length=14.5, radius=7.5, center=<-9.8473 -2.2348 -0.0004>, rotation=<rot 90 0 -8.1918>}]
 calf_r [{shapeType=Capsule, length=15.75, radius=7.75, center=<14.0229 0.4644 0>, rotation=<rot 90 0 172.7202>}]
 calf_twist_01_l [{shapeType=Capsule, length=22.25, radius=8.25, center=<1.8274 -1.1856 -0.0001>, rotation=<rot 90 90 100.625>}]
 calf_twist_01_r [{shapeType=Capsule, length=34, radius=7.5, center=<-12.3719 0.3795 0.0007>, rotation=<rot 90 180 -0.308>}]
 dyn_skirt_back_a_base_l [{shapeType=Sphere, radius=1.255}]
 dyn_skirt_back_a_base_r [{shapeType=Sphere, radius=1.255}]
 dyn_skirt_back_a_l [{shapeType=Box, extentX=19.5, extentY=4.5, extentZ=16.5, center=<5.925 -0.0186 0.001>, rotation=<rot 0 -5.5471 0>}]
 dyn_skirt_back_a_r [{shapeType=Box, extentX=19, extentY=4.5, extentZ=16, center=<-5.925 0.0186 -0.0001>, rotation=<rot 0 -5.5471 -180>}]
 dyn_skirt_back_b_l [{shapeType=Box, extentX=18.5, extentY=3.5, extentZ=16, center=<8.7465 -2.7645 0.0101>, rotation=<rot 0 -5.3129 0>}]
 dyn_skirt_back_b_r [
  {shapeType=Capsule, length=17, radius=11.5, center=<-9.1717 10.4263 -0.4904>, rotation=<rot -87.3807 -179.9991 175.5519>}
  {shapeType=Box, extentX=18.5, extentY=4.5, extentZ=16, center=<-8.5598 1.7064 -0.0001>, rotation=<rot 0 -5.3128 -180>}
 ]
 dyn_skirt_front_a_base_l [{shapeType=Sphere, radius=1.255}]
 dyn_skirt_front_a_base_r [{shapeType=Sphere, radius=1.255}]
 dyn_skirt_front_a_l [{shapeType=Box, extentX=14, extentY=3, extentZ=18, center=<3.9628 -0.0001 -0.1748>}]
 dyn_skirt_front_a_r [{shapeType=Box, extentX=14, extentY=4, extentZ=18, center=<-3.9778 0.0001 -0.0001>, rotation=<rot 0 0 -180>}]
 dyn_skirt_front_b_l [{shapeType=Box, extentX=20, extentY=4, extentZ=18.5, center=<8.671 0 -0.3838>}]
 dyn_skirt_front_b_r [{shapeType=Box, extentX=20, extentY=5, extentZ=19, center=<-8.7037 0 0>, rotation=<rot 0 0 -180>}]
 pelvis [{shapeType=Capsule, length=8, radius=10.75}]
 thigh_twist_01_l [{shapeType=Capsule, length=34.0182, radius=8.1822, rotation=<rot -0.0723 -89.3201 -90.8948>}]
 thigh_twist_01_r [{shapeType=Capsule, length=34.5182, radius=7.6822, rotation=<rot -0.0792 -89.7628 89.1049>}]
}
constraints [
 {
  bone1Name=thigh_twist_01_l
  bone2Name=pelvis
  swing1LimitDegrees=45
  swing1Motion=Free
  swing2LimitDegrees=45
  swing2Motion=Free
  twistLimitDegrees=45
  twistMotion=Free
 }
 {
  bone1Name=thigh_twist_01_r
  bone2Name=pelvis
  swing1LimitDegrees=45
  swing1Motion=Free
  swing2LimitDegrees=45
  swing2Motion=Free
  twistLimitDegrees=45
  twistMotion=Free
 }
 {
  bone1Name=dyn_skirt_front_a_base_r
  bone2Name=thigh_twist_01_r
  swing1LimitDegrees=45
  swing1Motion=Free
  swing2LimitDegrees=45
  swing2Motion=Free
  twistLimitDegrees=45
  twistMotion=Free
 }
 {
  bone1Name=dyn_skirt_back_a_base_r
  bone2Name=thigh_twist_01_r
  swing1LimitDegrees=45
  swing1Motion=Free
  swing2LimitDegrees=45
  swing2Motion=Free
  twistLimitDegrees=45
  twistMotion=Free
 }
 {
  bone1Name=dyn_skirt_front_a_base_l
  bone2Name=thigh_twist_01_l
  swing1LimitDegrees=45
  swing1Motion=Free
  swing2LimitDegrees=45
  swing2Motion=Free
  twistLimitDegrees=45
  twistMotion=Free
 }
 {
  bone1Name=dyn_skirt_back_a_base_l
  bone2Name=thigh_twist_01_l
  swing1LimitDegrees=45
  swing1Motion=Free
  swing2LimitDegrees=45
  swing2Motion=Free
  twistLimitDegrees=45
  twistMotion=Free
 }
 {bone1Name=dyn_skirt_front_a_l, bone2Name=dyn_skirt_front_a_base_l, swing1LimitDegrees=45, swing2LimitDegrees=15, twistLimitDegrees=5}
 {bone1Name=dyn_skirt_front_b_l, bone2Name=dyn_skirt_front_a_l, swing1LimitDegrees=45, swing2LimitDegrees=15, twistLimitDegrees=5}
 {bone1Name=dyn_skirt_front_a_r, bone2Name=dyn_skirt_front_a_base_r, swing1LimitDegrees=45, swing2LimitDegrees=15, twistLimitDegrees=5}
 {bone1Name=dyn_skirt_front_b_r, bone2Name=dyn_skirt_front_a_r, swing1LimitDegrees=45, swing2LimitDegrees=15, twistLimitDegrees=5}
 {bone1Name=dyn_skirt_back_a_l, bone2Name=dyn_skirt_back_a_base_l, swing1LimitDegrees=45, swing2LimitDegrees=20, twistLimitDegrees=5}
 {bone1Name=dyn_skirt_back_b_l, bone2Name=dyn_skirt_back_a_l, swing1LimitDegrees=80, swing2LimitDegrees=20, twistLimitDegrees=8}
 {bone1Name=dyn_skirt_back_a_r, bone2Name=dyn_skirt_back_a_base_r, swing1LimitDegrees=45, swing2LimitDegrees=20, twistLimitDegrees=5}
 {bone1Name=dyn_skirt_back_b_r, bone2Name=dyn_skirt_back_a_r, swing1LimitDegrees=80, swing2LimitDegrees=20, twistLimitDegrees=8}
 {
  bone1Name=calf_l
  bone2Name=pelvis
  swing1LimitDegrees=45
  swing1Motion=Free
  swing2LimitDegrees=45
  swing2Motion=Free
  twistLimitDegrees=45
  twistMotion=Free
 }
 {
  bone1Name=calf_r
  bone2Name=pelvis
  swing1LimitDegrees=45
  swing1Motion=Free
  swing2LimitDegrees=45
  swing2Motion=Free
  twistLimitDegrees=45
  twistMotion=Free
 }
 {
  bone1Name=calf_twist_01_l
  bone2Name=calf_l
  swing1LimitDegrees=45
  swing1Motion=Free
  swing2LimitDegrees=45
  swing2Motion=Free
  twistLimitDegrees=45
  twistMotion=Free
 }
 {
  bone1Name=calf_twist_01_r
  bone2Name=calf_r
  swing1LimitDegrees=45
  swing1Motion=Free
  swing2LimitDegrees=45
  swing2Motion=Free
  twistLimitDegrees=45
  twistMotion=Free
 }
 {
  bone1Name=dyn_skirt_back_a_l
  bone2Name=dyn_skirt_back_a_r
  swing1LimitDegrees=45
  swing1Motion=Locked
  swing2LimitDegrees=45
  swing2Motion=Locked
  twistLimitDegrees=45
  twistMotion=Locked
 }
]
