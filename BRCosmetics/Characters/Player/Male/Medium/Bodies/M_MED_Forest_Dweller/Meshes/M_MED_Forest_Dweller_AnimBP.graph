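# Animation Blueprint for the Forest Dweller male body cosmetic, adding secondary motion to its cloak, emblem, flaps and feathers.
# It copies the pose from the attached parent mesh, then re-anchors dyn parent and feather bones with weighted multi-bone constraints.
# AnimDynamics springs swing the cloak, emblem and arm flaps, with spherical limits around the spine on most cloak bones.
# It also uses RigidBody simulation and a crouch-weighted additive skirt bone offset.
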AnimBlueprint /BRCosmetics/Characters/Player/Male/Medium/Bodies/M_MED_Forest_Dweller/Meshes/M_MED_Forest_Dweller_AnimBP
target_skeleton=&/M_MED_Forest_Dweller_Skeleton
graphs [
 AnimGraph {nodes [
   "Transform (Modify) Bone - Bone: dyn_skirt_back_a_base_r"(Alpha=Get crouching() -> (crouching:double), Translation="10,0,0", Rotation="-10,0,0", Scale="1,1,1") {
    rotationMode=BMM_Additive
    rotationSpace=BCS_BoneSpace
    translationMode=BMM_Additive
    translationSpace=BCS_BoneSpace
    alphaBoolBlend {blendOption=Linear}
    alphaScaleBiasClamp {inRange {max=0, min=-110}, outRange {max=0, min=1}}
    boneToModify {boneName=dyn_skirt_back_a_base_r}
    rotation {pitch=-10}
    translation {x=10}
   }
   RigidBody() {
    bClampLinearTranslationLimitToRefPose=true
    bOverrideWorldGravity=true
    overridePhysicsAsset=&/M_MED_Forest_Dweller_Physics
    simulationSpace=BaseBoneSpace
    alphaBoolBlend {blendOption=Linear}
    baseBoneRef {boneName=root}
    componentAppliedLinearAccClamp=<1200 1200 1100>
    componentLinearAccScale=<0.7 0.7 1>
    componentLinearVelScale=<0.7 0.7 1>
    overrideWorldGravity {z=-980}
   }
   n1 = Get crouchMoveLeftLegR() -> (crouchMoveLeftLegR:double)
   n2 = Get crouchMoveRightLegL() -> (crouchMoveRightLegL:double)
   n3 = "Constraint - dyn_parent_farm_flap_l (PPPPPPPP)"!(
    ComponentPose=Local To Component(LocalPose=Copy Pose From Mesh() {bUseAttachedParent=true})
    Alpha=1
    ConstraintWeights_0=0.1388
    ConstraintWeights_1:float=0.1143
    ConstraintWeights_2:float=0.0361
    ConstraintWeights_3:float=0.071
    ConstraintWeights_4:float=0.0822
    ConstraintWeights_5:float=0.1627
    ConstraintWeights_6:float=0.0875
    ConstraintWeights_7:float=0.3074
   ) -> (Pose) {
    alphaBoolBlend {blendOption=Linear}
    boneToModify {boneName=dyn_parent_farm_flap_l}
    constraintSetup [
     {targetBone {boneName=clavicle_l}}
     {targetBone {boneName=upperarm_twist_02_l}}
     {targetBone {boneName=spine_02}}
     {targetBone {boneName=spine_03}}
     {targetBone {boneName=spine_04}}
     {targetBone {boneName=upperarm_twist_01_l}}
     {targetBone {boneName=dyn_farm_flap_02_l}}
     {targetBone {boneName=dyn_farm_flap_01_l}}
    ]
    constraintWeights [1,1,1,1,1,1,1,1]
   }
   n4 = "Constraint - dyn_parent_emblem (PPPPP)"!(
    ComponentPose=@n3
    Alpha=1
    ConstraintWeights_0=0.0124
    ConstraintWeights_1:float=0.0421
    ConstraintWeights_2:float=0.076
    ConstraintWeights_3:float=0.0488
    ConstraintWeights_4:float=0.8206
   ) -> (Pose) {
    alphaBoolBlend {blendOption=Linear}
    boneToModify {boneName=dyn_parent_emblem}
    constraintSetup [
     {targetBone {boneName=spine_01}}
     {targetBone {boneName=spine_02}}
     {targetBone {boneName=spine_03}}
     {targetBone {boneName=spine_04}}
     {targetBone {boneName=dyn_emblem_01}}
    ]
    constraintWeights [1,1,1,1,1]
   }
   n5 = Anim Dynamics - Bone: dyn_emblem_01(ComponentPose=@n4, Alpha=0.6, ExternalForce="0,0,0") {
    angularDampingOverride=1
    angularSpringConstant=1500
    bAngularSpring=true
    bGravityOverrideInSimSpace=true
    bOverrideAngularDamping=true
    bOverrideLinearDamping=true
    gravityScale=0.35
    linearDampingOverride=0.99
    simulationSpace=RootRelative
    alphaBoolBlend {blendOption=Linear}
    boundBone {boneName=dyn_emblem_01}
    componentLinearAccScale=<0.2 0.2 0.2>
    componentLinearVelScale=<0.2 0.2 0.2>
    physicsBodyDefinitions [{
      sphereCollisionRadius=0
      boundBone {boneName=dyn_emblem_01}
      boxExtents=<20 25 25>
      constraintSetup {angularLimitsMax=<0 35 8>, angularLimitsMin=<0 -35 -40>, angularTarget=<0 0 0>}
      localJointOffset=<10 0 0>
     }]
   }
   n6 = "Constraint - dyn_parent_cloak_l (PPPPPPPP)"!(
    ComponentPose=@n5
    Alpha=1
    ConstraintWeights_0=0.4855
    ConstraintWeights_1:float=0.0151
    ConstraintWeights_2:float=0.0205
    ConstraintWeights_3:float=0.216
    ConstraintWeights_4:float=0.0978
    ConstraintWeights_5:float=0.036
    ConstraintWeights_6:float=0.1138
    ConstraintWeights_7:float=0.0153
   ) -> (Pose) {
    alphaBoolBlend {blendOption=Linear}
    boneToModify {boneName=dyn_parent_cloak_l}
    constraintSetup [
     {targetBone {boneName=clavicle_l}}
     {targetBone {boneName=upperarm_twist_02_l}}
     {targetBone {boneName=spine_02}}
     {targetBone {boneName=spine_03}}
     {targetBone {boneName=spine_04}}
     {targetBone {boneName=spine_05}}
     {targetBone {boneName=upperarm_twist_01_l}}
     {targetBone {boneName=clavicle_r}}
    ]
    constraintWeights [1,1,1,1,1,1,1,1]
   }
   n7 = Anim Dynamics - Bone: dyn_cloak_bck_01_l(ComponentPose=@n6, Alpha=1, ExternalForce="0,0,0") {
    angularDampingOverride=1
    angularSpringConstant=1500
    bAngularSpring=true
    bGravityOverrideInSimSpace=true
    bOverrideAngularDamping=true
    bOverrideLinearDamping=true
    bUseSphericalLimits=true
    gravityScale=0.38
    linearDampingOverride=0.9999
    simulationSpace=RootRelative
    alphaBoolBlend {blendOption=Linear}
    boundBone {boneName=dyn_cloak_bck_01_l}
    componentAppliedLinearAccClamp=<1200 1200 1300>
    componentLinearAccScale=<0.4 0.4 0.35>
    componentLinearVelScale=<0.4 0.4 0.35>
    physicsBodyDefinitions [{
      collisionType=CustomSphere
      sphereCollisionRadius=3
      boundBone {boneName=dyn_cloak_bck_01_l}
      boxExtents=<10 8 10>
      constraintSetup {angularLimitsMax=<0 15 10>, angularLimitsMin=<0 -15 -40>, angularTarget=<0 0 0>}
      localJointOffset=<5 0 0>
     }]
    sphericalLimits [
     {limitRadius=17, drivingBone {boneName=spine_03}, sphereLocalOffset=<0 5 0>}
     {limitRadius=17.5, drivingBone {boneName=spine_04}, sphereLocalOffset=<0 4 0>}
     {limitRadius=16, drivingBone {boneName=spine_01}, sphereLocalOffset=<0 3 0>}
    ]
   }
   n8 = Anim Dynamics - Bone: dyn_cloak_bck_02_l(ComponentPose=@n7, Alpha=0.8, ExternalForce="0,0,0") {
    angularDampingOverride=1
    angularSpringConstant=1500
    bAngularSpring=true
    bGravityOverrideInSimSpace=true
    bOverrideAngularDamping=true
    bOverrideLinearDamping=true
    bUseSphericalLimits=true
    gravityScale=0.38
    linearDampingOverride=0.9999
    simulationSpace=RootRelative
    alphaBoolBlend {blendOption=Linear}
    boundBone {boneName=dyn_cloak_bck_02_l}
    componentAppliedLinearAccClamp=<1200 1200 1300>
    componentLinearAccScale=<0.4 0.4 0.35>
    componentLinearVelScale=<0.4 0.4 0.35>
    physicsBodyDefinitions [{
      collisionType=CustomSphere
      sphereCollisionRadius=2
      boundBone {boneName=dyn_cloak_bck_02_l}
      boxExtents=<10 8 10>
      constraintSetup {angularLimitsMax=<0 30 10>, angularLimitsMin=<0 -30 -40>, angularTarget=<0 0 0>}
      localJointOffset=<5 0 0>
     }]
    sphericalLimits [
     {limitRadius=18, drivingBone {boneName=spine_03}, sphereLocalOffset=<0 5 0>}
     {limitRadius=18, drivingBone {boneName=spine_04}, sphereLocalOffset=<0 3 0>}
     {limitRadius=16, drivingBone {boneName=spine_01}, sphereLocalOffset=<0 3 0>}
    ]
   }
   n9 = Anim Dynamics - Bone: dyn_cloak_bck_03_l(ComponentPose=@n8, Alpha=0.8, ExternalForce="0,0,0") {
    angularDampingOverride=1
    angularSpringConstant=1500
    bAngularSpring=true
    bGravityOverrideInSimSpace=true
    bOverrideAngularDamping=true
    bOverrideLinearDamping=true
    bUseSphericalLimits=true
    gravityScale=0.33
    linearDampingOverride=0.9999
    simulationSpace=RootRelative
    alphaBoolBlend {blendOption=Linear}
    boundBone {boneName=dyn_cloak_bck_03_l}
    componentAppliedLinearAccClamp=<1200 1200 1300>
    componentLinearAccScale=<0.4 0.4 0.35>
    componentLinearVelScale=<0.4 0.4 0.35>
    physicsBodyDefinitions [{
      collisionType=CustomSphere
      sphereCollisionRadius=5
      boundBone {boneName=dyn_cloak_bck_03_l}
      boxExtents=<12 8 10>
      constraintSetup {angularLimitsMax=<0 60 10>, angularLimitsMin=<0 -60 -40>, angularTarget=<1 0 0>}
      localJointOffset=<6 0 0>
     }]
    sphericalLimits [
     {limitRadius=18, drivingBone {boneName=spine_03}, sphereLocalOffset=<0 5 0>}
     {limitRadius=18, drivingBone {boneName=spine_04}, sphereLocalOffset=<0 3 0>}
     {limitRadius=16, drivingBone {boneName=spine_01}, sphereLocalOffset=<0 3 0>}
    ]
   }
   n10 = "Constraint - dyn_parent_cloak_r (PPPPPPPP)"!(
    ComponentPose=@n9
    Alpha=1
    ConstraintWeights_0=0.0187
    ConstraintWeights_1:float=0.1564
    ConstraintWeights_2:float=0.0348
    ConstraintWeights_3:float=0.025
    ConstraintWeights_4:float=0.1789
    ConstraintWeights_5:float=0.0727
    ConstraintWeights_6:float=0.0759
    ConstraintWeights_7:float=0.4377
   ) -> (Pose) {
    alphaBoolBlend {blendOption=Linear}
    boneToModify {boneName=dyn_parent_cloak_r}
    constraintSetup [
     {targetBone {boneName=clavicle_l}}
     {targetBone {boneName=upperarm_twist_01_r}}
     {targetBone {boneName=upperarm_twist_02_r}}
     {targetBone {boneName=spine_02}}
     {targetBone {boneName=spine_03}}
     {targetBone {boneName=spine_04}}
     {targetBone {boneName=spine_05}}
     {targetBone {boneName=clavicle_r}}
    ]
    constraintWeights [1,1,1,1,1,1,1,1]
   }
   n11 = Anim Dynamics - Bone: dyn_cloak_bck_01_l_r(ComponentPose=@n10, Alpha=0.75, ExternalForce="0,0,0") {
    angularDampingOverride=1
    angularSpringConstant=1500
    bAngularSpring=true
    bGravityOverrideInSimSpace=true
    bOverrideAngularDamping=true
    bOverrideLinearDamping=true
    bUseSphericalLimits=true
    gravityScale=0.38
    linearDampingOverride=0.9999
    simulationSpace=RootRelative
    alphaBoolBlend {blendOption=Linear}
    boundBone {boneName=dyn_cloak_bck_01_l_r}
    componentAppliedLinearAccClamp=<1200 1200 1300>
    componentLinearAccScale=<0.4 0.4 0.35>
    componentLinearVelScale=<0.4 0.4 0.35>
    physicsBodyDefinitions [{
      collisionType=CustomSphere
      sphereCollisionRadius=3
      boundBone {boneName=dyn_cloak_bck_01_l_r}
      boxExtents=<10 8 10>
      constraintSetup {angularLimitsMax=<0 15 10>, angularLimitsMin=<0 -15 -40>, angularTarget=<1 0 0>}
      localJointOffset=<5 0 0>
     }]
    sphericalLimits [
     {limitRadius=17, drivingBone {boneName=spine_03}, sphereLocalOffset=<0 5 0>}
     {limitRadius=17.5, drivingBone {boneName=spine_04}, sphereLocalOffset=<0 4 0>}
     {limitRadius=16, drivingBone {boneName=spine_01}, sphereLocalOffset=<0 3 0>}
    ]
   }
   n12 = Anim Dynamics - Bone: dyn_cloak_bck_02_l_r(ComponentPose=@n11, Alpha=0.8, ExternalForce="0,0,0") {
    angularDampingOverride=1
    angularSpringConstant=1500
    bAngularSpring=true
    bGravityOverrideInSimSpace=true
    bOverrideAngularDamping=true
    bOverrideLinearDamping=true
    bUseSphericalLimits=true
    gravityScale=0.38
    linearDampingOverride=0.9999
    simulationSpace=RootRelative
    alphaBoolBlend {blendOption=Linear}
    boundBone {boneName=dyn_cloak_bck_02_l_r}
    componentAppliedLinearAccClamp=<1200 1200 1300>
    componentLinearAccScale=<0.4 0.4 0.35>
    componentLinearVelScale=<0.4 0.4 0.35>
    physicsBodyDefinitions [{
      collisionType=CustomSphere
      sphereCollisionRadius=2
      boundBone {boneName=dyn_cloak_bck_02_l_r}
      boxExtents=<10 8 10>
      constraintSetup {angularLimitsMax=<0 30 10>, angularLimitsMin=<0 -30 -40>, angularTarget=<1 0 0>}
      localJointOffset=<5 0 0>
     }]
    sphericalLimits [
     {limitRadius=18, drivingBone {boneName=spine_03}, sphereLocalOffset=<0 5 0>}
     {limitRadius=18, drivingBone {boneName=spine_04}, sphereLocalOffset=<0 3 0>}
     {limitRadius=16, drivingBone {boneName=spine_01}, sphereLocalOffset=<0 3 0>}
    ]
   }
   n13 = Anim Dynamics - Bone: dyn_cloak_bck_03_l_r(ComponentPose=@n12, Alpha=0.8, ExternalForce="0,0,0") {
    angularDampingOverride=1
    angularSpringConstant=1500
    bAngularSpring=true
    bGravityOverrideInSimSpace=true
    bOverrideAngularDamping=true
    bOverrideLinearDamping=true
    bUseSphericalLimits=true
    gravityScale=0.33
    linearDampingOverride=0.9999
    simulationSpace=RootRelative
    alphaBoolBlend {blendOption=Linear}
    boundBone {boneName=dyn_cloak_bck_03_l_r}
    componentAppliedLinearAccClamp=<1200 1200 1300>
    componentLinearAccScale=<0.4 0.4 0.35>
    componentLinearVelScale=<0.4 0.4 0.35>
    physicsBodyDefinitions [{
      collisionType=CustomSphere
      sphereCollisionRadius=3
      boundBone {boneName=dyn_cloak_bck_03_l_r}
      boxExtents=<12 8 10>
      constraintSetup {angularLimitsMax=<0 50 10>, angularLimitsMin=<0 -50 -40>, angularTarget=<1 0 0>}
      localJointOffset=<6 0 0>
     }]
    sphericalLimits [
     {limitRadius=18, drivingBone {boneName=spine_03}, sphereLocalOffset=<0 5 0>}
     {limitRadius=18, drivingBone {boneName=spine_04}, sphereLocalOffset=<0 3 0>}
     {limitRadius=16, drivingBone {boneName=spine_01}, sphereLocalOffset=<0 3 0>}
    ]
   }
   n14 = "Constraint - dyn_parent_cloak_bck (PPPPPPPP)"!(
    ComponentPose=@n13
    Alpha=1
    ConstraintWeights_0=0.0643
    ConstraintWeights_1:float=0.1629
    ConstraintWeights_2:float=0.0009
    ConstraintWeights_3:float=0.0202
    ConstraintWeights_4:float=0.2636
    ConstraintWeights_5:float=0.1493
    ConstraintWeights_6:float=0.1981
    ConstraintWeights_7:float=0.1406
   ) -> (Pose) {
    alphaBoolBlend {blendOption=Linear}
    boneToModify {boneName=dyn_parent_cloak_bck}
    constraintSetup [
     {targetBone {boneName=dyn_cloak_bck_01}}
     {targetBone {boneName=clavicle_l}}
     {targetBone {boneName=upperarm_twist_01_r}}
     {targetBone {boneName=spine_02}}
     {targetBone {boneName=spine_03}}
     {targetBone {boneName=spine_04}}
     {targetBone {boneName=spine_05}}
     {targetBone {boneName=clavicle_r}}
    ]
    constraintWeights [1,1,1,1,1,1,1,1]
   }
   n15 = Anim Dynamics - Bone: dyn_cloak_bck_01(ComponentPose=@n14, Alpha=1, ExternalForce="0,0,0") {
    angularDampingOverride=1
    angularSpringConstant=1500
    bAngularSpring=true
    bGravityOverrideInSimSpace=true
    bOverrideAngularDamping=true
    bOverrideLinearDamping=true
    gravityScale=0.38
    linearDampingOverride=0.9999
    simulationSpace=RootRelative
    alphaBoolBlend {blendOption=Linear}
    boundBone {boneName=dyn_cloak_bck_01}
    componentAppliedLinearAccClamp=<1200 1200 1300>
    componentLinearAccScale=<0.4 0.4 0.35>
    componentLinearVelScale=<0.4 0.4 0.35>
    physicsBodyDefinitions [{
      collisionType=CustomSphere
      sphereCollisionRadius=3
      boundBone {boneName=dyn_cloak_bck_01}
      boxExtents=<10 8 10>
      constraintSetup {angularLimitsMax=<0 15 11>, angularLimitsMin=<0 -15 -40>, angularTarget=<1 0 0>}
      localJointOffset=<5 0 0>
     }]
    sphericalLimits [
     {limitRadius=17, drivingBone {boneName=spine_03}, sphereLocalOffset=<0 5 0>}
     {limitRadius=17.5, drivingBone {boneName=spine_04}, sphereLocalOffset=<0 4 0>}
     {limitRadius=16, drivingBone {boneName=spine_01}, sphereLocalOffset=<0 3 0>}
    ]
   }
   n16 = Anim Dynamics - Bone: dyn_cloak_bck_02(ComponentPose=@n15, Alpha=1, ExternalForce="0,0,0") {
    angularDampingOverride=1
    angularSpringConstant=1500
    bAngularSpring=true
    bGravityOverrideInSimSpace=true
    bOverrideAngularDamping=true
    bOverrideLinearDamping=true
    bUseSphericalLimits=true
    gravityScale=0.38
    linearDampingOverride=0.9999
    simulationSpace=RootRelative
    alphaBoolBlend {blendOption=Linear}
    boundBone {boneName=dyn_cloak_bck_02}
    componentAppliedLinearAccClamp=<1200 1200 1300>
    componentLinearAccScale=<0.4 0.4 0.35>
    componentLinearVelScale=<0.4 0.4 0.35>
    physicsBodyDefinitions [{
      collisionType=CustomSphere
      sphereCollisionRadius=2
      boundBone {boneName=dyn_cloak_bck_02}
      boxExtents=<10 8 10>
      constraintSetup {angularLimitsMax=<0 30 10>, angularLimitsMin=<0 -30 -40>, angularTarget=<1 0 0>}
      localJointOffset=<5 0 0>
     }]
    sphericalLimits [
     {limitRadius=17, drivingBone {boneName=spine_03}, sphereLocalOffset=<0 5 0>}
     {limitRadius=17.5, drivingBone {boneName=spine_04}, sphereLocalOffset=<0 4 0>}
     {limitRadius=16, drivingBone {boneName=spine_01}, sphereLocalOffset=<0 3 0>}
    ]
   }
   n17 = Anim Dynamics - Bone: dyn_cloak_bck_03(ComponentPose=@n16, Alpha=1, ExternalForce="0,0,0") {
    angularDampingOverride=1
    angularSpringConstant=1500
    bAngularSpring=true
    bGravityOverrideInSimSpace=true
    bOverrideAngularDamping=true
    bOverrideLinearDamping=true
    bUseSphericalLimits=true
    gravityScale=0.33
    linearDampingOverride=0.9999
    simulationSpace=RootRelative
    alphaBoolBlend {blendOption=Linear}
    boundBone {boneName=dyn_cloak_bck_03}
    componentAppliedLinearAccClamp=<1200 1200 1300>
    componentLinearAccScale=<0.4 0.4 0.35>
    componentLinearVelScale=<0.4 0.4 0.35>
    physicsBodyDefinitions [{
      collisionType=CustomSphere
      sphereCollisionRadius=3
      boundBone {boneName=dyn_cloak_bck_03}
      boxExtents=<12 8 10>
      constraintSetup {angularLimitsMax=<0 60 10>, angularLimitsMin=<0 -60 -40>, angularTarget=<1 0 0>}
      localJointOffset=<6 0 0>
     }]
    sphericalLimits [
     {limitRadius=18, drivingBone {boneName=spine_03}, sphereLocalOffset=<0 5 0>}
     {limitRadius=18, drivingBone {boneName=spine_04}, sphereLocalOffset=<0 3 0>}
     {limitRadius=16, drivingBone {boneName=spine_01}, sphereLocalOffset=<0 3 0>}
    ]
   }
   n18 = "Constraint - dyn_parent_arm_flap_r (PPPPP)"!(
    ComponentPose=@n17
    Alpha=1
    ConstraintWeights_0=0.4961
    ConstraintWeights_1:float=0.1808
    ConstraintWeights_2:float=0.1037
    ConstraintWeights_3:float=0.0116
    ConstraintWeights_4:float=0.2079
   ) -> (Pose) {
    alphaBoolBlend {blendOption=Linear}
    boneToModify {boneName=dyn_parent_arm_flap_r}
    constraintSetup [
     {targetBone {boneName=upperarm_twist_01_r}}
     {targetBone {boneName=clavicle_r}}
     {targetBone {boneName=upperarm_r}}
     {targetBone {boneName=spine_03}}
     {targetBone {boneName=upperarm_twist_02_r}}
    ]
    constraintWeights [1,1,1,1,1]
   }
   n19 = Anim Dynamics - Bone: dyn_arm_flap_01_r(ComponentPose=@n18, Alpha=0.9, ExternalForce="0,0,0") {
    angularDampingOverride=1
    angularSpringConstant=1500
    bAngularSpring=true
    bGravityOverrideInSimSpace=true
    bOverrideAngularDamping=true
    gravityScale=0.35
    simulationSpace=RootRelative
    alphaBoolBlend {blendOption=Linear}
    boundBone {boneName=dyn_arm_flap_01_r}
    componentAppliedLinearAccClamp=<1200 1200 1100>
    componentLinearAccScale=<0.8 0.8 0.4>
    componentLinearVelScale=<0.8 0.8 0.8>
    physicsBodyDefinitions [{
      sphereCollisionRadius=0
      boundBone {boneName=dyn_arm_flap_01_r}
      boxExtents=<10 20 20>
      constraintSetup {angularLimitsMax=<0 0 25>, angularLimitsMin=<0 0 0>, angularTarget=<0 0 0>}
      localJointOffset=<-10 0 0>
     }]
   }
   n20 = Anim Dynamics - Bone: dyn_arm_flap_02_r(ComponentPose=@n19, Alpha=0.7, ExternalForce="0,0,0") {
    angularDampingOverride=1
    angularSpringConstant=1500
    bAngularSpring=true
    bGravityOverrideInSimSpace=true
    bOverrideAngularDamping=true
    gravityScale=0.35
    simulationSpace=RootRelative
    alphaBoolBlend {blendOption=Linear}
    boundBone {boneName=dyn_arm_flap_02_r}
    componentAppliedLinearAccClamp=<1200 1200 1100>
    componentLinearAccScale=<0.8 0.8 0.4>
    componentLinearVelScale=<0.8 0.8 0.8>
    physicsBodyDefinitions [{
      sphereCollisionRadius=0
      boundBone {boneName=dyn_arm_flap_02_r}
      boxExtents=<10 15 20>
      constraintSetup {angularLimitsMax=<0 0 25>, angularLimitsMin=<0 0 0>, angularTarget=<0 0 0>}
      localJointOffset=<-5 0 0>
     }]
   }
   n21 = Anim Dynamics - Bone: dyn_arm_flap_03_r(ComponentPose=@n20, Alpha=0.1, ExternalForce="0,0,0") {
    angularDampingOverride=1
    angularSpringConstant=1500
    bAngularSpring=true
    bGravityOverrideInSimSpace=true
    bOverrideAngularDamping=true
    gravityScale=0.35
    simulationSpace=RootRelative
    alphaBoolBlend {blendOption=Linear}
    boundBone {boneName=dyn_arm_flap_03_r}
    componentAppliedLinearAccClamp=<1200 1200 1100>
    componentLinearAccScale=<0.8 0.8 0.8>
    componentLinearVelScale=<0.8 0.8 0.8>
    physicsBodyDefinitions [{
      sphereCollisionRadius=0
      boundBone {boneName=dyn_arm_flap_03_r}
      boxExtents=<20 20 20>
      constraintSetup {angularLimitsMax=<0 0 25>, angularLimitsMin=<0 0 0>, angularTarget=<0 0 0>}
      localJointOffset=<-13 0 0>
     }]
   }
   n22 = "Constraint - dyn_parent_arm_flap_l (PPPPPP)"!(
    ComponentPose=@n21
    Alpha=1
    ConstraintWeights_0=0.2077
    ConstraintWeights_1:float=0.2156
    ConstraintWeights_2:float=0.0193
    ConstraintWeights_3:float=0.0178
    ConstraintWeights_4:float=0.3947
    ConstraintWeights_5:float=0.1448
   ) -> (Pose) {
    alphaBoolBlend {blendOption=Linear}
    boneToModify {boneName=dyn_parent_arm_flap_l}
    constraintSetup [
     {targetBone {boneName=clavicle_l}}
     {targetBone {boneName=upperarm_twist_02_l}}
     {targetBone {boneName=spine_03}}
     {targetBone {boneName=spine_05}}
     {targetBone {boneName=upperarm_twist_01_l}}
     {targetBone {boneName=upperarm_l}}
    ]
    constraintWeights [1,1,1,1,1,1]
   }
   n23 = Anim Dynamics - Bone: dyn_arm_flap_01_l(ComponentPose=@n22, Alpha=0.9, ExternalForce="0,0,0") {
    angularDampingOverride=1
    angularSpringConstant=1500
    bAngularSpring=true
    bGravityOverrideInSimSpace=true
    bOverrideAngularDamping=true
    gravityScale=0.35
    simulationSpace=RootRelative
    alphaBoolBlend {blendOption=Linear}
    boundBone {boneName=dyn_arm_flap_01_l}
    componentAppliedLinearAccClamp=<1200 1200 1100>
    componentLinearAccScale=<0.8 0.8 0.4>
    componentLinearVelScale=<0.8 0.8 0.8>
    physicsBodyDefinitions [{
      sphereCollisionRadius=0
      boundBone {boneName=dyn_arm_flap_01_l}
      boxExtents=<10 20 20>
      constraintSetup {angularLimitsMax=<0 0 0>, angularLimitsMin=<0 0 -40>, angularTarget=<0 0 0>}
      localJointOffset=<10 0 0>
     }]
   }
   n24 = Anim Dynamics - Bone: dyn_arm_flap_02_l(ComponentPose=@n23, Alpha=0.7, ExternalForce="0,0,0") {
    angularDampingOverride=1
    angularSpringConstant=1500
    bAngularSpring=true
    bGravityOverrideInSimSpace=true
    bOverrideAngularDamping=true
    bOverrideLinearDamping=true
    gravityScale=0.35
    linearDampingOverride=0.9999
    simulationSpace=RootRelative
    alphaBoolBlend {blendOption=Linear}
    boundBone {boneName=dyn_arm_flap_02_l}
    componentAppliedLinearAccClamp=<1200 1200 1100>
    componentLinearAccScale=<0.8 0.8 0.4>
    componentLinearVelScale=<0.8 0.8 0.8>
    physicsBodyDefinitions [{
      sphereCollisionRadius=0
      boundBone {boneName=dyn_arm_flap_02_l}
      boxExtents=<10 20 20>
      constraintSetup {angularLimitsMax=<0 10 0>, angularLimitsMin=<0 -10 -30>, angularTarget=<0 0 0>}
      localJointOffset=<10 0 0>
     }]
   }
   n25 = Anim Dynamics - Bone: dyn_arm_flap_03_l(ComponentPose=@n24, Alpha=0.1, ExternalForce="0,0,0") {
    angularDampingOverride=1
    angularSpringConstant=1500
    bAngularSpring=true
    bGravityOverrideInSimSpace=true
    bOverrideAngularDamping=true
    bOverrideLinearDamping=true
    gravityScale=0.321
    linearDampingOverride=0.9999
    simulationSpace=RootRelative
    alphaBoolBlend {blendOption=Linear}
    boundBone {boneName=dyn_arm_flap_03_l}
    componentAppliedLinearAccClamp=<1200 1200 1100>
    componentLinearAccScale=<0.8 0.8 0.8>
    componentLinearVelScale=<0.8 0.8 0.8>
    physicsBodyDefinitions [{
      sphereCollisionRadius=0
      boundBone {boneName=dyn_arm_flap_03_l}
      boxExtents=<20 20 20>
      constraintSetup {angularLimitsMax=<0 10 0>, angularLimitsMin=<0 -10 -20>, angularTarget=<0 0 0>}
      localJointOffset=<13 0 0>
     }]
   }
   n26 = "Constraint - dyn_bck_feathA_01 (PPPPPPPP)"!(
    ComponentPose=@n25
    Alpha=1
    ConstraintWeights_0=0.0025
    ConstraintWeights_1:float=0.1067
    ConstraintWeights_2:float=0.0286
    ConstraintWeights_3:float=0.0038
    ConstraintWeights_4:float=0.6175
    ConstraintWeights_5:float=0.0616
    ConstraintWeights_6:float=0.1369
    ConstraintWeights_7:float=0.0425
   ) -> (Pose) {
    alphaBoolBlend {blendOption=Linear}
    boneToModify {boneName=dyn_bck_feathA_01}
    constraintSetup [
     {targetBone {boneName=dyn_feathJ_02_l}}
     {targetBone {boneName=clavicle_l}}
     {targetBone {boneName=spine_03}}
     {targetBone {boneName=spine_04}}
     {targetBone {boneName=spine_05}}
     {targetBone {boneName=clavicle_r}}
     {targetBone {boneName=neck_01}}
     {targetBone {boneName=dyn_bck_feathA_02}}
    ]
    constraintWeights [1,1,1,1,1,1,1,1]
   }
   n27 = "Constraint - dyn_bck_feathB_01 (PPPPPPPP)"!(
    ComponentPose=@n26
    Alpha=1
    ConstraintWeights_0=0.1693
    ConstraintWeights_1:float=0.0367
    ConstraintWeights_2:float=0.0708
    ConstraintWeights_3:float=0.0113
    ConstraintWeights_4:float=0.3225
    ConstraintWeights_5:float=0.2841
    ConstraintWeights_6:float=0.0637
    ConstraintWeights_7:float=0.0416
   ) -> (Pose) {
    alphaBoolBlend {blendOption=Linear}
    boneToModify {boneName=dyn_bck_feathB_01}
    constraintSetup [
     {targetBone {boneName=clavicle_l}}
     {targetBone {boneName=dyn_feathI_02_l}}
     {targetBone {boneName=spine_03}}
     {targetBone {boneName=spine_04}}
     {targetBone {boneName=spine_05}}
     {targetBone {boneName=dyn_bck_feathB_02}}
     {targetBone {boneName=clavicle_r}}
     {targetBone {boneName=neck_01}}
    ]
    constraintWeights [1,1,1,1,1,1,1,1]
   }
   n28 = "Constraint - dyn_bck_feathC_01 (PPPPPPPP)"!(
    ComponentPose=@n27
    Alpha=1
    ConstraintWeights_0=0.0417
    ConstraintWeights_1:float=0.1197
    ConstraintWeights_2:float=0.086
    ConstraintWeights_3:float=0.0222
    ConstraintWeights_4:float=0.0092
    ConstraintWeights_5:float=0.5049
    ConstraintWeights_6:float=0.109
    ConstraintWeights_7:float=0.1073
   ) -> (Pose) {
    alphaBoolBlend {blendOption=Linear}
    boneToModify {boneName=dyn_bck_feathC_01}
    constraintSetup [
     {targetBone {boneName=clavicle_l}}
     {targetBone {boneName=dyn_bck_feathC_02}}
     {targetBone {boneName=dyn_bck_feathD_02}}
     {targetBone {boneName=spine_03}}
     {targetBone {boneName=spine_04}}
     {targetBone {boneName=spine_05}}
     {targetBone {boneName=clavicle_r}}
     {targetBone {boneName=neck_01}}
    ]
    constraintWeights [1,1,1,1,1,1,1,1]
   }
   n29 = "Constraint - dyn_bck_feathD_01 (PPPPPPPP)"!(
    ComponentPose=@n28
    Alpha=1
    ConstraintWeights_0=0.0417
    ConstraintWeights_1:float=0.015
    ConstraintWeights_2:float=0.0318
    ConstraintWeights_3:float=0.0234
    ConstraintWeights_4:float=0.019
    ConstraintWeights_5:float=0.597
    ConstraintWeights_6:float=0.1488
    ConstraintWeights_7:float=0.1234
   ) -> (Pose) {
    alphaBoolBlend {blendOption=Linear}
    boneToModify {boneName=dyn_bck_feathD_01}
    constraintSetup [
     {targetBone {boneName=clavicle_l}}
     {targetBone {boneName=dyn_bck_feathC_02}}
     {targetBone {boneName=dyn_bck_feathD_02}}
     {targetBone {boneName=spine_03}}
     {targetBone {boneName=spine_04}}
     {targetBone {boneName=spine_05}}
     {targetBone {boneName=clavicle_r}}
     {targetBone {boneName=neck_01}}
    ]
    constraintWeights [1,1,1,1,1,1,1,1]
   }
   n30 = "Constraint - dyn_bck_feathE_01 (PPPPPPPP)"!(
    ComponentPose=@n29
    Alpha=1
    ConstraintWeights_0=0.0571
    ConstraintWeights_1:float=0.0126
    ConstraintWeights_2:float=0.0686
    ConstraintWeights_3:float=0.0022
    ConstraintWeights_4:float=0.544
    ConstraintWeights_5:float=0.0145
    ConstraintWeights_6:float=0.225
    ConstraintWeights_7:float=0.0759
   ) -> (Pose) {
    alphaBoolBlend {blendOption=Linear}
    boneToModify {boneName=dyn_bck_feathE_01}
    constraintSetup [
     {targetBone {boneName=clavicle_l}}
     {targetBone {boneName=dyn_bck_feathE_02}}
     {targetBone {boneName=spine_03}}
     {targetBone {boneName=spine_04}}
     {targetBone {boneName=spine_05}}
     {targetBone {boneName=dyn_bck_feathB_02}}
     {targetBone {boneName=clavicle_r}}
     {targetBone {boneName=neck_01}}
    ]
    constraintWeights [1,1,1,1,1,1,1,1]
   }
   n31 = "Constraint - dyn_feathA_01_l (PPPPPPPP)"!(
    ComponentPose=@n30
    Alpha=1
    ConstraintWeights_0=0.1793
    ConstraintWeights_1:float=0.0154
    ConstraintWeights_2:float=0.3167
    ConstraintWeights_3:float=0.0176
    ConstraintWeights_4:float=0.0921
    ConstraintWeights_5:float=0.1735
    ConstraintWeights_6:float=0.1769
    ConstraintWeights_7:float=0.0285
   ) -> (Pose) {
    alphaBoolBlend {blendOption=Linear}
    boneToModify {boneName=dyn_feathA_01_l}
    constraintSetup [
     {targetBone {boneName=clavicle_l}}
     {targetBone {boneName=dyn_feathB_02_l}}
     {targetBone {boneName=dyn_feathA_02_l}}
     {targetBone {boneName=spine_02}}
     {targetBone {boneName=spine_03}}
     {targetBone {boneName=spine_04}}
     {targetBone {boneName=spine_05}}
     {targetBone {boneName=clavicle_r}}
    ]
    constraintWeights [1,1,1,1,1,1,1,1]
   }
   n32 = "Constraint - dyn_feathB_01_l (PPPPPPPP)"!(
    ComponentPose=@n31
    Alpha=1
    ConstraintWeights_0=0.2757
    ConstraintWeights_1:float=0.0466
    ConstraintWeights_2:float=0.0769
    ConstraintWeights_3:float=0.0153
    ConstraintWeights_4:float=0.0841
    ConstraintWeights_5:float=0.1923
    ConstraintWeights_6:float=0.2801
    ConstraintWeights_7:float=0.0289
   ) -> (Pose) {
    alphaBoolBlend {blendOption=Linear}
    boneToModify {boneName=dyn_feathB_01_l}
    constraintSetup [
     {targetBone {boneName=clavicle_l}}
     {targetBone {boneName=dyn_feathB_02_l}}
     {targetBone {boneName=dyn_feathA_02_l}}
     {targetBone {boneName=spine_02}}
     {targetBone {boneName=spine_03}}
     {targetBone {boneName=spine_04}}
     {targetBone {boneName=spine_05}}
     {targetBone {boneName=clavicle_r}}
    ]
    constraintWeights [1,1,1,1,1,1,1,1]
   }
   n33 = "Constraint - dyn_feathC_01_l (PPPPPPPP)"!(
    ComponentPose=@n32
    Alpha=1
    ConstraintWeights_0=0.034
    ConstraintWeights_1:float=0.3456
    ConstraintWeights_2:float=0.004
    ConstraintWeights_3:float=0.0499
    ConstraintWeights_4:float=0.1296
    ConstraintWeights_5:float=0.3838
    ConstraintWeights_6:float=0.0266
    ConstraintWeights_7:float=0.0265
   ) -> (Pose) {
    alphaBoolBlend {blendOption=Linear}
    boneToModify {boneName=dyn_feathC_01_l}
    constraintSetup [
     {targetBone {boneName=dyn_feathC_02_l}}
     {targetBone {boneName=clavicle_l}}
     {targetBone {boneName=dyn_feathB_02_l}}
     {targetBone {boneName=spine_03}}
     {targetBone {boneName=spine_04}}
     {targetBone {boneName=spine_05}}
     {targetBone {boneName=clavicle_r}}
     {targetBone {boneName=neck_01}}
    ]
    constraintWeights [1,1,1,1,1,1,1,1]
   }
   n34 = "Constraint - dyn_feathD_01_l (PPPPPP)"!(
    ComponentPose=@n33
    Alpha=1
    ConstraintWeights_0=0.3712
    ConstraintWeights_1:float=0.0218
    ConstraintWeights_2:float=0.0683
    ConstraintWeights_3:float=0.4754
    ConstraintWeights_4:float=0.0153
    ConstraintWeights_5:float=0.048
   ) -> (Pose) {
    alphaBoolBlend {blendOption=Linear}
    boneToModify {boneName=dyn_feathD_01_l}
    constraintSetup [
     {targetBone {boneName=clavicle_l}}
     {targetBone {boneName=spine_03}}
     {targetBone {boneName=spine_04}}
     {targetBone {boneName=spine_05}}
     {targetBone {boneName=upperarm_twist_01_l}}
     {targetBone {boneName=neck_01}}
    ]
    constraintWeights [1,1,1,1,1,1]
   }
   n35 = "Constraint - dyn_feathE_01_l (PPPPPPPP)"!(
    ComponentPose=@n34
    Alpha=1
    ConstraintWeights_0=0.0006
    ConstraintWeights_1:float=0.4204
    ConstraintWeights_2:float=0.0601
    ConstraintWeights_3:float=0.0272
    ConstraintWeights_4:float=0.0132
    ConstraintWeights_5:float=0.4283
    ConstraintWeights_6:float=0.0117
    ConstraintWeights_7:float=0.0386
   ) -> (Pose) {
    alphaBoolBlend {blendOption=Linear}
    boneToModify {boneName=dyn_feathE_01_l}
    constraintSetup [
     {targetBone {boneName=dyn_feathD_02_l}}
     {targetBone {boneName=clavicle_l}}
     {targetBone {boneName=dyn_feathE_02_l}}
     {targetBone {boneName=spine_03}}
     {targetBone {boneName=spine_04}}
     {targetBone {boneName=spine_05}}
     {targetBone {boneName=upperarm_twist_01_l}}
     {targetBone {boneName=neck_01}}
    ]
    constraintWeights [1,1,1,1,1,1,1,1]
   }
   n36 = "Constraint - dyn_feathF_01_l (PPPP)"!(
    ComponentPose=@n35
    Alpha=1
    ConstraintWeights_0=0.0528
    ConstraintWeights_1:float=0.0283
    ConstraintWeights_2:float=0.5053
    ConstraintWeights_3:float=0.4135
   ) -> (Pose) {
    alphaBoolBlend {blendOption=Linear}
    boneToModify {boneName=dyn_feathF_01_l}
    constraintSetup [{targetBone {boneName=neck_01}}, {targetBone {boneName=spine_03}}, {targetBone {boneName=spine_05}}, {targetBone {boneName=clavicle_l}}]
    constraintWeights [1,1,1,1]
   }
   n37 = "Constraint - dyn_feathG_01_l (PPPPPPPP)"!(
    ComponentPose=@n36
    Alpha=1
    ConstraintWeights_0=0.0102
    ConstraintWeights_1:float=0.3844
    ConstraintWeights_2:float=0.0538
    ConstraintWeights_3:float=0.0119
    ConstraintWeights_4:float=0.2945
    ConstraintWeights_5:float=0.0084
    ConstraintWeights_6:float=0.0195
    ConstraintWeights_7:float=0.2173
   ) -> (Pose) {
    alphaBoolBlend {blendOption=Linear}
    boneToModify {boneName=dyn_feathG_01_l}
    constraintSetup [
     {targetBone {boneName=dyn_feathF_02_l}}
     {targetBone {boneName=clavicle_l}}
     {targetBone {boneName=spine_03}}
     {targetBone {boneName=spine_04}}
     {targetBone {boneName=spine_05}}
     {targetBone {boneName=clavicle_r}}
     {targetBone {boneName=neck_01}}
     {targetBone {boneName=dyn_feathG_02_l}}
    ]
    constraintWeights [1,1,1,1,1,1,1,1]
   }
   n38 = "Constraint - dyn_feathH_01_l (PPPPPPPP)"!(
    ComponentPose=@n37
    Alpha=1
    ConstraintWeights_0=0.3666
    ConstraintWeights_1:float=0.0998
    ConstraintWeights_2:float=0.0387
    ConstraintWeights_3:float=0.0888
    ConstraintWeights_4:float=0.0365
    ConstraintWeights_5:float=0.3123
    ConstraintWeights_6:float=0.0331
    ConstraintWeights_7:float=0.0242
   ) -> (Pose) {
    alphaBoolBlend {blendOption=Linear}
    boneToModify {boneName=dyn_feathH_01_l}
    constraintSetup [
     {targetBone {boneName=clavicle_l}}
     {targetBone {boneName=dyn_feathH_02_l}}
     {targetBone {boneName=dyn_feathI_02_l}}
     {targetBone {boneName=spine_03}}
     {targetBone {boneName=spine_04}}
     {targetBone {boneName=spine_05}}
     {targetBone {boneName=clavicle_r}}
     {targetBone {boneName=dyn_feathG_02_l}}
    ]
    constraintWeights [1,1,1,1,1,1,1,1]
   }
   n39 = "Constraint - dyn_feathI_01_l (PPPPPPPP)"!(
    ComponentPose=@n38
    Alpha=1
    ConstraintWeights_0=0.2466
    ConstraintWeights_1:float=0.0144
    ConstraintWeights_2:float=0.1315
    ConstraintWeights_3:float=0.0817
    ConstraintWeights_4:float=0.0134
    ConstraintWeights_5:float=0.4008
    ConstraintWeights_6:float=0.0614
    ConstraintWeights_7:float=0.0501
   ) -> (Pose) {
    alphaBoolBlend {blendOption=Linear}
    boneToModify {boneName=dyn_feathI_01_l}
    constraintSetup [
     {targetBone {boneName=clavicle_l}}
     {targetBone {boneName=dyn_feathH_02_l}}
     {targetBone {boneName=dyn_feathI_02_l}}
     {targetBone {boneName=spine_03}}
     {targetBone {boneName=spine_04}}
     {targetBone {boneName=spine_05}}
     {targetBone {boneName=clavicle_r}}
     {targetBone {boneName=neck_01}}
    ]
    constraintWeights [1,1,1,1,1,1,1,1]
   }
   n40 = "Constraint - dyn_feathJ_01_l (PPPPPP)"!(
    ComponentPose=@n39
    Alpha=1
    ConstraintWeights_0=0.1661
    ConstraintWeights_1:float=0.027
    ConstraintWeights_2:float=0.033
    ConstraintWeights_3:float=0.6146
    ConstraintWeights_4:float=0.0358
    ConstraintWeights_5:float=0.1235
   ) -> (Pose) {
    alphaBoolBlend {blendOption=Linear}
    boneToModify {boneName=dyn_feathJ_01_l}
    constraintSetup [
     {targetBone {boneName=clavicle_l}}
     {targetBone {boneName=spine_03}}
     {targetBone {boneName=spine_04}}
     {targetBone {boneName=spine_05}}
     {targetBone {boneName=clavicle_r}}
     {targetBone {boneName=neck_01}}
    ]
    constraintWeights [1,1,1,1,1,1]
   }
   n41 = "Constraint - dyn_feathA_01_r (PPPPPPPP)"!(
    ComponentPose=@n40
    Alpha=1
    ConstraintWeights_0=0.0981
    ConstraintWeights_1:float=0.0015
    ConstraintWeights_2:float=0.0168
    ConstraintWeights_3:float=0.0625
    ConstraintWeights_4:float=0.2215
    ConstraintWeights_5:float=0.3321
    ConstraintWeights_6:float=0.2503
    ConstraintWeights_7:float=0.0172
   ) -> (Pose) {
    alphaBoolBlend {blendOption=Linear}
    boneToModify {boneName=dyn_feathA_01_r}
    constraintSetup [
     {targetBone {boneName=dyn_feathA_02_r}}
     {targetBone {boneName=clavicle_l}}
     {targetBone {boneName=spine_02}}
     {targetBone {boneName=spine_03}}
     {targetBone {boneName=spine_04}}
     {targetBone {boneName=spine_05}}
     {targetBone {boneName=clavicle_r}}
     {targetBone {boneName=neck_01}}
    ]
    constraintWeights [1,1,1,1,1,1,1,1]
   }
   n42 = "Constraint - dyn_feathB_01_r (PPPPP)"!(
    ComponentPose=@n41
    Alpha=1
    ConstraintWeights_0=0.2727
    ConstraintWeights_1:float=0.0264
    ConstraintWeights_2:float=0.0536
    ConstraintWeights_3:float=0.2296
    ConstraintWeights_4:float=0.4176
   ) -> (Pose) {
    alphaBoolBlend {blendOption=Linear}
    boneToModify {boneName=dyn_feathB_01_r}
    constraintSetup [
     {targetBone {boneName=clavicle_r}}
     {targetBone {boneName=neck_01}}
     {targetBone {boneName=spine_03}}
     {targetBone {boneName=spine_04}}
     {targetBone {boneName=spine_05}}
    ]
    constraintWeights [1,1,1,1,1]
   }
   n43 = "Constraint - dyn_feathC_01_r (PPPPPPPP)"!(
    ComponentPose=@n42
    Alpha=1
    ConstraintWeights_0=0.0059
    ConstraintWeights_1:float=0.0026
    ConstraintWeights_2:float=0.2659
    ConstraintWeights_3:float=0.0249
    ConstraintWeights_4:float=0.1043
    ConstraintWeights_5:float=0.3318
    ConstraintWeights_6:float=0.2362
    ConstraintWeights_7:float=0.0284
   ) -> (Pose) {
    alphaBoolBlend {blendOption=Linear}
    boneToModify {boneName=dyn_feathC_01_r}
    constraintSetup [
     {targetBone {boneName=clavicle_l}}
     {targetBone {boneName=upperarm_twist_01_r}}
     {targetBone {boneName=dyn_feathC_02_r}}
     {targetBone {boneName=spine_03}}
     {targetBone {boneName=spine_04}}
     {targetBone {boneName=spine_05}}
     {targetBone {boneName=clavicle_r}}
     {targetBone {boneName=neck_01}}
    ]
    constraintWeights [1,1,1,1,1,1,1,1]
   }
   n44 = "Constraint - dyn_feathD_01_r (PPPPPPPP)"!(
    ComponentPose=@n43
    Alpha=1
    ConstraintWeights_0=0.0111
    ConstraintWeights_1:float=0.0281
    ConstraintWeights_2:float=0.0258
    ConstraintWeights_3:float=0.0927
    ConstraintWeights_4:float=0.4254
    ConstraintWeights_5:float=0.3522
    ConstraintWeights_6:float=0.038
    ConstraintWeights_7:float=0.0267
   ) -> (Pose) {
    alphaBoolBlend {blendOption=Linear}
    boneToModify {boneName=dyn_feathD_01_r}
    constraintSetup [
     {targetBone {boneName=clavicle_l}}
     {targetBone {boneName=dyn_feathC_02_r}}
     {targetBone {boneName=spine_03}}
     {targetBone {boneName=spine_04}}
     {targetBone {boneName=spine_05}}
     {targetBone {boneName=clavicle_r}}
     {targetBone {boneName=neck_01}}
     {targetBone {boneName=dyn_feathD_02_r}}
    ]
    constraintWeights [1,1,1,1,1,1,1,1]
   }
   n45 = "Constraint - dyn_feathE_01_r (PPPPPPPP)"!(
    ComponentPose=@n44
    Alpha=1
    ConstraintWeights_0=0.0165
    ConstraintWeights_1:float=0.0637
    ConstraintWeights_2:float=0.0287
    ConstraintWeights_3:float=0.0221
    ConstraintWeights_4:float=0.3675
    ConstraintWeights_5:float=0.473
    ConstraintWeights_6:float=0.0267
    ConstraintWeights_7:float=0.0018
   ) -> (Pose) {
    alphaBoolBlend {blendOption=Linear}
    boneToModify {boneName=dyn_feathE_01_r}
    constraintSetup [
     {targetBone {boneName=upperarm_twist_01_r}}
     {targetBone {boneName=dyn_feathE_02_r}}
     {targetBone {boneName=spine_03}}
     {targetBone {boneName=spine_04}}
     {targetBone {boneName=spine_05}}
     {targetBone {boneName=clavicle_r}}
     {targetBone {boneName=neck_01}}
     {targetBone {boneName=dyn_feathF_02_r}}
    ]
    constraintWeights [1,1,1,1,1,1,1,1]
   }
   n46 = "Constraint - dyn_feathF_01_r (PPPPPPPP)"!(
    ComponentPose=@n45
    Alpha=1
    ConstraintWeights_0=0.0075
    ConstraintWeights_1:float=0.0246
    ConstraintWeights_2:float=0.0244
    ConstraintWeights_3:float=0.4318
    ConstraintWeights_4:float=0.4162
    ConstraintWeights_5:float=0.0038
    ConstraintWeights_6:float=0.0414
    ConstraintWeights_7:float=0.0503
   ) -> (Pose) {
    alphaBoolBlend {blendOption=Linear}
    boneToModify {boneName=dyn_feathF_01_r}
    constraintSetup [
     {targetBone {boneName=upperarm_twist_01_r}}
     {targetBone {boneName=spine_03}}
     {targetBone {boneName=spine_04}}
     {targetBone {boneName=spine_05}}
     {targetBone {boneName=clavicle_r}}
     {targetBone {boneName=dyn_feathH_02_r}}
     {targetBone {boneName=neck_01}}
     {targetBone {boneName=dyn_feathF_02_r}}
    ]
    constraintWeights [1,1,1,1,1,1,1,1]
   }
   n47 = "Constraint - dyn_feathG_01_r (PPPP)"!(
    ComponentPose=@n46
    Alpha=1
    ConstraintWeights_0=0.5245
    ConstraintWeights_1:float=0.0302
    ConstraintWeights_2:float=0.0481
    ConstraintWeights_3:float=0.3972
   ) -> (Pose) {
    alphaBoolBlend {blendOption=Linear}
    boneToModify {boneName=dyn_feathG_01_r}
    constraintSetup [{targetBone {boneName=clavicle_r}}, {targetBone {boneName=neck_01}}, {targetBone {boneName=spine_03}}, {targetBone {boneName=spine_05}}]
    constraintWeights [1,1,1,1]
   }
   n48 = "Constraint - dyn_feathH_01_r (PPPPP)"!(
    ComponentPose=@n47
    Alpha=1
    ConstraintWeights_0=0.3448
    ConstraintWeights_1:float=0.0757
    ConstraintWeights_2:float=0.0241
    ConstraintWeights_3:float=0.0167
    ConstraintWeights_4:float=0.5387
   ) -> (Pose) {
    alphaBoolBlend {blendOption=Linear}
    boneToModify {boneName=dyn_feathH_01_r}
    constraintSetup [
     {targetBone {boneName=clavicle_r}}
     {targetBone {boneName=neck_01}}
     {targetBone {boneName=spine_03}}
     {targetBone {boneName=spine_04}}
     {targetBone {boneName=spine_05}}
    ]
    constraintWeights [1,1,1,1,1]
   }
   n49 = "Constraint - dyn_feathI_01_r (PPPPPPPP)"!(
    ComponentPose=@n48
    Alpha=1
    ConstraintWeights_0=0.0121
    ConstraintWeights_1:float=0.0117
    ConstraintWeights_2:float=0.0812
    ConstraintWeights_3:float=0.0689
    ConstraintWeights_4:float=0.0028
    ConstraintWeights_5:float=0.3406
    ConstraintWeights_6:float=0.4606
    ConstraintWeights_7:float=0.0222
   ) -> (Pose) {
    alphaBoolBlend {blendOption=Linear}
    boneToModify {boneName=dyn_feathI_01_r}
    constraintSetup [
     {targetBone {boneName=clavicle_l}}
     {targetBone {boneName=upperarm_twist_01_r}}
     {targetBone {boneName=dyn_feathI_02_r}}
     {targetBone {boneName=spine_03}}
     {targetBone {boneName=spine_04}}
     {targetBone {boneName=spine_05}}
     {targetBone {boneName=clavicle_r}}
     {targetBone {boneName=neck_01}}
    ]
    constraintWeights [1,1,1,1,1,1,1,1]
   }
   n50 = "Constraint - dyn_feathJ_01_r (PPPPPP)"!(
    ComponentPose=@n49
    Alpha=1
    ConstraintWeights_0=0.0141
    ConstraintWeights_1:float=0.0222
    ConstraintWeights_2:float=0.0194
    ConstraintWeights_3:float=0.5736
    ConstraintWeights_4:float=0.2766
    ConstraintWeights_5:float=0.0941
   ) -> (Pose) {
    alphaBoolBlend {blendOption=Linear}
    boneToModify {boneName=dyn_feathJ_01_r}
    constraintSetup [
     {targetBone {boneName=clavicle_l}}
     {targetBone {boneName=spine_03}}
     {targetBone {boneName=spine_04}}
     {targetBone {boneName=spine_05}}
     {targetBone {boneName=clavicle_r}}
     {targetBone {boneName=neck_01}}
    ]
    constraintWeights [1,1,1,1,1,1]
   }
   n51 = "Constraint - dyn_feathK_01_r (PPPPPPP)"!(
    ComponentPose=@n50
    Alpha=1
    ConstraintWeights_0=0.0399
    ConstraintWeights_1:float=0.0139
    ConstraintWeights_2:float=0.0956
    ConstraintWeights_3:float=0.0133
    ConstraintWeights_4:float=0.4111
    ConstraintWeights_5:float=0.3924
    ConstraintWeights_6:float=0.0338
   ) -> (Pose) {
    alphaBoolBlend {blendOption=Linear}
    boneToModify {boneName=dyn_feathK_01_r}
    constraintSetup [
     {targetBone {boneName=clavicle_l}}
     {targetBone {boneName=upperarm_twist_01_r}}
     {targetBone {boneName=spine_03}}
     {targetBone {boneName=spine_04}}
     {targetBone {boneName=spine_05}}
     {targetBone {boneName=clavicle_r}}
     {targetBone {boneName=neck_01}}
    ]
    constraintWeights [1,1,1,1,1,1,1]
   }
   n52 = "Transform (Modify) Bone - Bone: dyn_skirt_back_a_base_l"(ComponentPose=@n51, Alpha=Get legUp_L() -> (legUp_L:double), Translation="-5,0,0", Rotation="0,-80,0", Scale="1,1,1") {
    rotationMode=BMM_Additive
    rotationSpace=BCS_BoneSpace
    translationMode=BMM_Additive
    translationSpace=BCS_BoneSpace
    alphaBoolBlend {blendOption=Linear}
    alphaScaleBiasClamp {bMapRange=true, inRange {max=0, min=-110}, outRange {max=0, min=1}}
    boneToModify {boneName=dyn_skirt_back_a_base_l}
    rotation {yaw=-80}
    translation {x=-5}
   }
   n53 = "Transform (Modify) Bone - Bone: dyn_skirt_back_a_base_r"(ComponentPose=@n52, Alpha=Get legUp_R() -> (legUp_R:double), Translation="-5,0,0", Rotation="0,-80,0", Scale="1,1,1") {
    rotationMode=BMM_Additive
    rotationSpace=BCS_BoneSpace
    translationMode=BMM_Additive
    translationSpace=BCS_BoneSpace
    alphaBoolBlend {blendOption=Linear}
    alphaScaleBiasClamp {bMapRange=true, inRange {max=0, min=-110}, outRange {max=0, min=1}}
    boneToModify {boneName=dyn_skirt_back_a_base_r}
    rotation {yaw=-80}
    translation {x=-5}
   }
   n54 = "Transform (Modify) Bone - Bone: dyn_skirt_back_a_base_l"(ComponentPose=@n53, Alpha=@n2, Translation="0,0,5", Rotation="100,0,0", Scale="1,1,1") {
    rotationMode=BMM_Additive
    rotationSpace=BCS_BoneSpace
    translationMode=BMM_Additive
    translationSpace=BCS_BoneSpace
    alphaBoolBlend {blendOption=Linear}
    alphaScaleBiasClamp {inRange {max=0, min=-110}, outRange {max=0, min=1}}
    boneToModify {boneName=dyn_skirt_back_a_base_l}
    rotation {pitch=100}
    translation {z=5}
   }
   n55 = "Transform (Modify) Bone - Bone: dyn_skirt_back_a_base_r"(ComponentPose=@n54, Alpha=@n2, Translation="0,0,0", Rotation="50,0,0", Scale="1,1,1") {
    rotationMode=BMM_Additive
    rotationSpace=BCS_BoneSpace
    translationMode=BMM_Additive
    translationSpace=BCS_BoneSpace
    alphaBoolBlend {blendOption=Linear}
    alphaScaleBiasClamp {inRange {max=0, min=-110}, outRange {max=0, min=1}}
    boneToModify {boneName=dyn_skirt_back_a_base_r}
    rotation {pitch=50}
   }
   n56 = "Transform (Modify) Bone - Bone: dyn_skirt_back_a_base_r"(ComponentPose=@n55, Alpha=@n1, Translation="0,0,-5", Rotation="100,0,0", Scale="1,1,1") {
    rotationMode=BMM_Additive
    rotationSpace=BCS_BoneSpace
    translationMode=BMM_Additive
    translationSpace=BCS_BoneSpace
    alphaBoolBlend {blendOption=Linear}
    alphaScaleBiasClamp {inRange {max=0, min=-110}, outRange {max=0, min=1}}
    boneToModify {boneName=dyn_skirt_back_a_base_r}
    rotation {pitch=100}
    translation {z=-5}
   }
   n57 = "Transform (Modify) Bone - Bone: dyn_skirt_back_a_base_l"(ComponentPose=@n56, Alpha=@n1, Translation="0,0,0", Rotation="50,0,0", Scale="1,1,1") {
    rotationMode=BMM_Additive
    rotationSpace=BCS_BoneSpace
    translationMode=BMM_Additive
    translationSpace=BCS_BoneSpace
    alphaBoolBlend {blendOption=Linear}
    alphaScaleBiasClamp {inRange {max=0, min=-110}, outRange {max=0, min=1}}
    boneToModify {boneName=dyn_skirt_back_a_base_l}
    rotation {pitch=50}
   }
   n58 = RigidBody(ComponentPose=@n57) {
    bClampLinearTranslationLimitToRefPose=true
    bOverrideWorldGravity=true
    overridePhysicsAsset=&/M_MED_Forest_Dweller_Skirt_Physics
    simulationSpace=BaseBoneSpace
    alphaBoolBlend {blendOption=Linear}
    baseBoneRef {boneName=root}
    overrideWorldGravity {z=-980}
   }
   n59 = RigidBody(ComponentPose=@n58) {
    bClampLinearTranslationLimitToRefPose=true
    bOverrideWorldGravity=true
    overridePhysicsAsset=&/M_MED_Forest_Dweller_Feathers_Physics
    simulationSpace=BaseBoneSpace
    alphaBoolBlend {blendOption=Linear}
    baseBoneRef {boneName=root}
    componentAppliedLinearAccClamp=<1200 1200 1200>
    componentLinearAccScale=<0.7 0.7 0.35>
    componentLinearVelScale=<0.7 0.7 0.35>
    overrideWorldGravity {z=-850}
   }
   Output Pose(Result=Component To Local(ComponentPose=@n59)) {layerGroup=None}
   Reroute Node(InputPin:Wildcard) -> (OutputPin:Wildcard)
  ]}
 EventGraph {nodes [
   Event BlueprintUpdateAnimation()
   > n1 = Sequence!(execute) -> (then_0, then_1, then_2:Exec, then_3:Exec)
   n2 = GetOwnerSkeletalMeshComponent()
   n3 = GetDeltaTransformFromRefPose(self=@n2, BoneName=thigh_l, BaseName=pelvis)
   > Set legUp_L(legUp_L:double=BreakRotator(InRot=BreakTransform(InTransform=@n3).Rotation).Yaw) -> (Output_Get:double)
   n4 = GetDeltaTransformFromRefPose(self=@n2, BoneName=thigh_r, BaseName=pelvis)
   Set legUp_R(execute=@n1.then_1, legUp_R:double=BreakRotator(InRot=BreakTransform(InTransform=@n4).Rotation).Yaw) -> (Output_Get:double)
   n5 = "Not Equal (Boolean)"!(A:bool=Get bIsMoving2D(), B:bool=true) -> (ReturnValue:bool) {is=title K2Node_CallFunction}
   n6 = Get bIsCrouching()
   Set crouching(execute=@n1.then_2, crouching:double="To Float (Boolean)"(InBool=AND Boolean(A=@n6, B=@n5))) -> (Output_Get:double)
   n7 = Sequence!(execute=@n1.then_3) -> (then_0, then_1, then_2)
   n8 = GetOwnerDirection!(self:Custom Character Part Anim Instance Object Reference) -> (ReturnValue:float)
   > Set IsMovingL(IsMovingL:bool=AND Boolean(
     A="float > float"(A=@n8, B=-120) {is=title K2Node_CallFunction}
     B="float < float"!(A:double=@n8, B:double=-60) -> (ReturnValue:bool) {is=title K2Node_CallFunction}
    )) -> (Output_Get:bool)
   n9 = Get bIsCrouching()
   > n10 = Branch(Condition=AND Boolean(A=Get IsMovingL() -> (IsMovingL:bool), B=@n9))
   n11 = BreakTransform(InTransform=GetDeltaTransformFromRefPose(self=@n2, BoneName=thigh_r, BaseName=pelvis))
   > Set crouchMoveLeftLegR(crouchMoveLeftLegR:double=MapRangeClamped(Value=BreakRotator(InRot=@n11.Rotation).Pitch, InRangeA=-15, InRangeB=0, OutRangeA=1, OutRangeB=0)) -> (Output_Get:double)
   Set crouchMoveLeftLegR(execute=@n10.else, crouchMoveLeftLegR=0)
   Set IsMovingR(execute=@n7.then_1, IsMovingR:bool=AND Boolean(A="float > float"(A=@n8, B=60), B="float < float"(A=@n8, B=120))) -> (Output_Get:bool)
   > n12 = Branch(Condition=AND Boolean(A=@n9, B=Get IsMovingR() -> (IsMovingR:bool)))
   n13 = BreakTransform(InTransform=GetDeltaTransformFromRefPose(self=@n2, BoneName=thigh_l, BaseName=pelvis))
   > Set crouchMoveRightLegL(crouchMoveRightLegL:double=MapRangeClamped(Value=BreakRotator(InRot=@n13.Rotation).Pitch, InRangeA=-15, InRangeB=0, OutRangeA=1, OutRangeB=0)) -> (Output_Get:double)
   Set crouchMoveRightLegL(execute=@n12.else, crouchMoveRightLegL=0)
   TryGetPawnOwner()
  ]}
]
variables [
 crouching {type=number}
 crouchMoveLeftLegR {type=number}
 crouchMoveRightLegL {type=number}
 IsMovingL {type=boolean}
 IsMovingR {type=boolean}
 legUp_L {type=number}
 legUp_R {type=number}
]
