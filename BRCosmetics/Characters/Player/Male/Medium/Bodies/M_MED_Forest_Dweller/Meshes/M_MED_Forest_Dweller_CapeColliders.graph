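# Physics asset of collision shapes for the cape on the Forest Dweller medium male player body, as its CapeColliders name indicates.
# Capsules cover pelvis, spine, neck and clavicles, with a sphere on the head and multiple overlapping capsules on spine_02 and spine_05.
# Extra capsules ride the dynamic back cloak and back skirt bones.
# Limb bodies have no shapes and there are no constraints, so it defines colliders rather than a jointed ragdoll.
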PhysicsAsset /BRCosmetics/Characters/Player/Male/Medium/Bodies/M_MED_Forest_Dweller/Meshes/M_MED_Forest_Dweller_CapeColliders
bodies {
 calf_l []
 calf_r []
 clavicle_l [{shapeType=Capsule, length=7.7636, radius=11.1427, center=<9.9642 -0.7741 -1.5923>, rotation=<rot 75.9352 -0.0001 175.5622>}]
 clavicle_r [{shapeType=Capsule, length=7.7636, radius=11.1427, center=<-9.9642 0.7741 1.5923>, rotation=<rot 75.9351 -0.0001 -4.4378>}]
 dyn_cloak_bck_03 [{shapeType=Capsule, length=11, radius=7.75, shapeName=dyn_cloak_bck_03_capsule, center=<-3.7177 0 0>, rotation=<rot -90 -179.578 179.5782>}]
 dyn_cloak_bck_03_l [{
   shapeType=Capsule
   length=11
   radius=7.75
   shapeName=dyn_cloak_bck_03_l_capsule
   center=<-2.431 0.6393 4.7871>
   rotation=<rot -40 -179.9996 180>
  }]
 dyn_cloak_bck_03_l_r [{shapeType=Capsule, length=11, radius=7.75, shapeName=dyn_cloak_bck_03_l_r_capsule, center=<-0.9754 0 -3.7838>, rotation=<rot -30 0.001 0>}]
 dyn_skirt_back_b_l [{shapeType=Capsule, length=5.75, radius=16.75, shapeName=dyn_skirt_back_b_l_capsule, center=<0 4.2186 0>, rotation=<rot 90 179.8419 179.8416>}]
 dyn_skirt_back_b_r [{
   shapeType=Capsule
   length=5.75
   radius=16.75
   shapeName=dyn_skirt_back_b_r_capsule
   center=<0 -4.8526 0>
   rotation=<rot 90 179.4936 179.4933>
  }]
 foot_l []
 foot_r []
 head [{shapeType=Sphere, radius=17.3125, center=<11.028 -0.174 0>}]
 lowerarm_l []
 lowerarm_r []
 neck_02 [{shapeType=Capsule, length=13.4375, radius=7.3175, center=<-0.2642 -0.3408 0>, rotation=<rot 90 -171.1193 -175.3978>}]
 pelvis [{shapeType=Capsule, length=8.1562, radius=16.2888, center=<-2.6215 -1.3695 0>}]
 spine_01 [{shapeType=Capsule, length=9.2187, radius=15.1638, center=<-2.6215 -0.7156 0>}]
 spine_02 [
  {shapeType=Capsule, length=10.5937, radius=10.945, center=<-2.6215 0.4993 0.0001>}
  {shapeType=Capsule, length=29.875, radius=11.1013, center=<2.5024 0.4992 3.9825>, rotation=<rot 90 0.0049 0.0049>}
  {shapeType=Capsule, length=29.875, radius=11.1013, center=<2.4704 0.4991 -3.9825>, rotation=<rot 90 -0.0396 -0.0396>}
 ]
 spine_03 [{shapeType=Capsule, length=4.6562, radius=15.2575, center=<-1.953 -1.7923 0.0001>}]
 spine_04 [{shapeType=Capsule, length=12.1562, radius=16.1325, center=<2.1467 -2.5536 0.0002>}]
 spine_05 [
  {shapeType=Capsule, length=14.0937, radius=20.945, center=<4.3506 0.6328 0>}
  {shapeType=Capsule, length=12.7812, radius=17.0625, center=<0.7875 3.1143 0>}
 ]
 thigh_l []
 thigh_r []
 upperarm_l []
 upperarm_r []
}
constraints []
