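# Physics asset for the flame part of the male medium Forest Dweller character cosmetic, giving its dynamic hat and flame bones simulation bodies.
# Four small capsules form a serial chain from dyn_parent_hat through dyn_parent_flame and dyn_flame_01 to dyn_flame_02.
# All three joints swing freely; the hat-to-flame joint also twists freely, while the two flame joints carry a much narrower twist limit angle.
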PhysicsAsset /BRCosmetics/Characters/Player/Male/Medium/Bodies/M_MED_Forest_Dweller/Meshes/Parts/M_MED_Forest_Dweller_Flame_Physics
bodies {
 dyn_flame_01 [{shapeType=Capsule, length=1, radius=1, rotation=<rot 90 0 0>}]
 dyn_flame_02 [{shapeType=Capsule, length=1, radius=1, rotation=<rot 90 0 0>}]
 dyn_parent_flame [{shapeType=Capsule, length=0.505, radius=0.505, rotation=<rot 0 -90 0>}]
 dyn_parent_hat [{shapeType=Capsule, length=0.505, radius=0.505, rotation=<rot 0 -90 0>}]
}
constraints [
 {
  bone1Name=dyn_parent_flame
  bone2Name=dyn_parent_hat
  swing1LimitDegrees=45
  swing1Motion=Free
  swing2LimitDegrees=45
  swing2Motion=Free
  twistLimitDegrees=45
  twistMotion=Free
 }
 {
  bone1Name=dyn_flame_01
  bone2Name=dyn_parent_flame
  swing1LimitDegrees=45
  swing1Motion=Free
  swing2LimitDegrees=45
  swing2Motion=Free
  twistLimitDegrees=5
 }
 {
  bone1Name=dyn_flame_02
  bone2Name=dyn_flame_01
  swing1LimitDegrees=45
  swing1Motion=Free
  swing2LimitDegrees=45
  swing2Motion=Free
  twistLimitDegrees=5
 }
]
