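# Physics asset for the hat part of the male medium Forest Dweller player cosmetic, defining bodies for the hat's dynamics bones.
# A large capsule covers dyn_hat_01, while a tiny capsule on dyn_parent_hat serves as its anchor.
# One constraint joins the two with tight swing and twist limits of a few degrees, so the hat can only tilt slightly relative to its parent.
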PhysicsAsset /BRCosmetics/Characters/Player/Male/Medium/Bodies/M_MED_Forest_Dweller/Meshes/Parts/M_MED_Forest_Dweller_Hat_Physics
bodies {
 dyn_hat_01 [{shapeType=Capsule, length=18, radius=18, center=<0.15 0 0>, rotation=<rot 0 0 -90>}]
 dyn_parent_hat [{shapeType=Capsule, length=0.505, radius=0.505, rotation=<rot 0 -90 0>}]
}
constraints [{bone1Name=dyn_hat_01, bone2Name=dyn_parent_hat, swing1LimitDegrees=5, swing2LimitDegrees=3, twistLimitDegrees=2}]
